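# Physics asset for the male medium BraveBuild player body, set up for a dynamic jacket that hangs from the pelvis.
# Seven two-segment capsule chains cover the front, sides and back, each starting at a small base body locked rotationally to the pelvis.
# Chain joints lock twist and carry per-joint swing angles, which differ between some left and right pieces.
# A pelvis sphere plus spine and thigh capsules provide body volumes around the jacket.
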PhysicsAsset /BRCosmetics/Characters/Player/Male/Medium/Bodies/M_MED_BraveBuild/Meshes/M_MED_BraveBuild_Physics
bodies {
 dyn_jacket_01_L [{
   shapeType=Capsule
   length=13
   radius=3
   shapeName=dyn_jacket_01_L_capsule
   center=<7.8188 0.4658 -2.2695>
   rotation=<rot -0.0004 -0.0015 74.9997>
  }]
 dyn_jacket_01_R [{
   shapeType=Capsule
   length=13
   radius=3
   shapeName=dyn_jacket_01_R_capsule
   center=<7.8188 -0.4658 -2.2695>
   rotation=<rot -0.0002 0.0007 105.0001>
  }]
 dyn_jacket_02_L [{
   shapeType=Capsule
   length=13
   radius=3.5
   shapeName=dyn_jacket_02_L_capsule
   center=<12.6837 0 -2.4726>
   rotation=<rot -0.0001 -0.0004 89.9999>
  }]
 dyn_jacket_02_R [{
   shapeType=Capsule
   length=13
   radius=2.5
   shapeName=dyn_jacket_02_R_capsule
   center=<12.6837 -0.2224 -2.1873>
   rotation=<rot -0.0002 0.0007 95.0001>
  }]
 dyn_jacket_back_01 [{shapeType=Capsule, length=13, radius=4.75, shapeName=dyn_jacket_back_01_capsule, center=<7.8188 0 -1.109>, rotation=<rot 0 0 90>}]
 dyn_jacket_back_01_L [{shapeType=Capsule, length=13, radius=4.75, shapeName=dyn_jacket_back_01_L_capsule, center=<7.8188 0.274 -1.5542>, rotation=<rot 0 0 80>}]
 dyn_jacket_back_01_R [{shapeType=Capsule, length=13, radius=4.75, shapeName=dyn_jacket_back_01_R_capsule, center=<7.8188 0.3573 -1.5542>, rotation=<rot 0 0 100>}]
 dyn_jacket_back_02 [{shapeType=Capsule, length=13, radius=4.5, shapeName=dyn_jacket_back_02_capsule, center=<7.6301 0 -0.6252>, rotation=<rot 0 0 90>}]
 dyn_jacket_back_02_L [{
   shapeType=Capsule
   length=13
   radius=3.5
   shapeName=dyn_jacket_back_02_L_capsule
   center=<12.6837 -0.2844 1.6132>
   rotation=<rot 0.0002 -0.0002 80>
  }]
 dyn_jacket_back_02_R [{
   shapeType=Capsule
   length=13
   radius=3.5
   shapeName=dyn_jacket_back_02_R_capsule
   center=<12.6837 0.2844 1.6132>
   rotation=<rot 0.0001 0.0002 100>
  }]
 dyn_jacket_front_01_L [{shapeType=Capsule, length=12.5, radius=3.5, shapeName=dyn_jacket_front_01_L_capsule, center=<7.613 -1.2374 -2.2152>, rotation=<rot 0 0 110>}]
 dyn_jacket_front_01_R [{
   shapeType=Capsule
   length=12.5
   radius=3
   shapeName=dyn_jacket_front_01_R_capsule
   center=<7.613 0.7106 -2.3792>
   rotation=<rot -0.0002 0.0002 70>
  }]
 dyn_jacket_front_02_L [{
   shapeType=Capsule
   length=12
   radius=4.5
   shapeName=dyn_jacket_front_02_L_capsule
   center=<11.6834 -0.891 -3.7646>
   rotation=<rot -0.0001 -0.0001 105>
  }]
 dyn_jacket_front_02_R [{
   shapeType=Capsule
   length=12
   radius=4
   shapeName=dyn_jacket_front_02_R_capsule
   center=<11.6834 0.8886 -3.7466>
   rotation=<rot -0.0003 0.0003 75>
  }]
 jacket_back_base [{shapeType=Capsule, length=0, radius=0.5, shapeName=jacket_back_base_capsule}]
 jacket_back_base_L [{shapeType=Capsule, length=0, radius=0.5, shapeName=jacket_back_base_L_capsule}]
 jacket_back_base_R [{shapeType=Capsule, length=0, radius=0.5, shapeName=jacket_back_base_R_capsule}]
 jacket_base_L [{shapeType=Capsule, length=0, radius=0.5, shapeName=jacket_base_L_capsule}]
 jacket_base_R [{shapeType=Capsule, length=0, radius=0.5, shapeName=jacket_base_R_capsule}]
 jacket_front_base_L [{shapeType=Capsule, length=0, radius=0.5, shapeName=jacket_front_base_L_capsule}]
 jacket_front_base_R [{shapeType=Capsule, length=0, radius=0.5, shapeName=jacket_front_base_R_capsule, rotation=<rot 0.0001 0 -70>}]
 pelvis [{shapeType=Sphere, radius=9.75, shapeName=pelvis_sphere}]
 spine_01 [{shapeType=Capsule, length=14.5, radius=6, shapeName=spine_01_capsule, center=<-4.2909 0 0.3>}]
 spine_02 [{shapeType=Capsule, length=11.5, radius=15, shapeName=spine_02_capsule, center=<6.7934 -0.8622 0>, rotation=<rot -90 0 11.6862>}]
 thigh_l [{shapeType=Capsule, length=25, radius=9.2, center=<-13 0 0>, rotation=<rot 0 -88.8961 -88.4765>}]
 thigh_r [{shapeType=Capsule, length=25, radius=9.2, center=<13 0 0>, rotation=<rot 0 -88.8961 -88.4765>}]
}
constraints [
 {
  bone1Name=pelvis
  bone2Name=jacket_back_base
  swing1LimitDegrees=45
  swing1Motion=Locked
  swing2LimitDegrees=45
  swing2Motion=Locked
  twistLimitDegrees=45
  twistMotion=Locked
 }
 {
  bone1Name=pelvis
  bone2Name=jacket_back_base_L
  swing1LimitDegrees=45
  swing1Motion=Locked
  swing2LimitDegrees=45
  swing2Motion=Locked
  twistLimitDegrees=45
  twistMotion=Locked
 }
 {
  bone1Name=pelvis
  bone2Name=jacket_base_L
  swing1LimitDegrees=45
  swing1Motion=Locked
  swing2LimitDegrees=45
  swing2Motion=Locked
  twistLimitDegrees=45
  twistMotion=Locked
 }
 {
  bone1Name=pelvis
  bone2Name=jacket_front_base_L
  swing1LimitDegrees=45
  swing1Motion=Locked
  swing2LimitDegrees=45
  swing2Motion=Locked
  twistLimitDegrees=45
  twistMotion=Locked
 }
 {
  bone1Name=pelvis
  bone2Name=jacket_back_base_R
  swing1LimitDegrees=45
  swing1Motion=Locked
  swing2LimitDegrees=45
  swing2Motion=Locked
  twistLimitDegrees=45
  twistMotion=Locked
 }
 {
  bone1Name=pelvis
  bone2Name=jacket_base_R
  swing1LimitDegrees=45
  swing1Motion=Locked
  swing2LimitDegrees=45
  swing2Motion=Locked
  twistLimitDegrees=45
  twistMotion=Locked
 }
 {
  bone1Name=jacket_back_base_R
  bone2Name=dyn_jacket_back_01_R
  swing1LimitDegrees=5
  swing2LimitDegrees=70
  twistLimitDegrees=5
  twistMotion=Locked
 }
 {bone1Name=jacket_base_R, bone2Name=dyn_jacket_01_R, swing1LimitDegrees=10, swing2LimitDegrees=30, twistLimitDegrees=5, twistMotion=Locked}
 {
  bone1Name=jacket_front_base_L
  bone2Name=dyn_jacket_front_01_L
  swing1LimitDegrees=10
  swing2LimitDegrees=30
  twistLimitDegrees=5
  twistMotion=Locked
 }
 {
  bone1Name=jacket_front_base_R
  bone2Name=dyn_jacket_front_01_R
  swing1LimitDegrees=10
  swing2LimitDegrees=30
  twistLimitDegrees=0
  twistMotion=Locked
 }
 {
  bone1Name=jacket_back_base
  bone2Name=dyn_jacket_back_01
  swing1LimitDegrees=5
  swing2LimitDegrees=70
  twistLimitDegrees=5
  twistMotion=Locked
 }
 {
  bone1Name=jacket_back_base_L
  bone2Name=dyn_jacket_back_01_L
  swing1LimitDegrees=10
  swing2LimitDegrees=30
  twistLimitDegrees=5
  twistMotion=Locked
 }
 {bone1Name=jacket_base_L, bone2Name=dyn_jacket_01_L, swing1LimitDegrees=5, swing2LimitDegrees=70, twistLimitDegrees=5, twistMotion=Locked}
 {
  bone1Name=pelvis
  bone2Name=jacket_front_base_R
  swing1LimitDegrees=45
  swing1Motion=Locked
  swing2LimitDegrees=45
  swing2Motion=Locked
  twistLimitDegrees=45
  twistMotion=Locked
 }
 {
  bone1Name=dyn_jacket_front_02_R
  bone2Name=dyn_jacket_front_01_R
  swing1LimitDegrees=10
  swing2LimitDegrees=20
  twistLimitDegrees=5
  twistMotion=Locked
 }
 {bone1Name=dyn_jacket_02_R, bone2Name=dyn_jacket_01_R, swing1LimitDegrees=10, swing2LimitDegrees=20, twistLimitDegrees=5, twistMotion=Locked}
 {bone1Name=dyn_jacket_02_L, bone2Name=dyn_jacket_01_L, swing1LimitDegrees=10, swing2LimitDegrees=40, twistLimitDegrees=10, twistMotion=Locked}
 {
  bone1Name=dyn_jacket_back_02_R
  bone2Name=dyn_jacket_back_01_R
  swing1LimitDegrees=10
  swing2LimitDegrees=30
  twistLimitDegrees=10
  twistMotion=Locked
 }
 {
  bone1Name=dyn_jacket_back_02_L
  bone2Name=dyn_jacket_back_01_L
  swing1LimitDegrees=10
  swing2LimitDegrees=20
  twistLimitDegrees=5
  twistMotion=Locked
 }
 {
  bone1Name=dyn_jacket_back_02
  bone2Name=dyn_jacket_back_01
  swing1LimitDegrees=10
  swing2LimitDegrees=30
  twistLimitDegrees=10
  twistMotion=Locked
 }
 {
  bone1Name=dyn_jacket_front_02_L
  bone2Name=dyn_jacket_front_01_L
  swing1LimitDegrees=10
  swing2LimitDegrees=20
  twistLimitDegrees=5
  twistMotion=Locked
 }
]
